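; BPI_DropPodSpawner is a Blueprint interface for spawning drop pods on military rank-ups.
; It declares one function, SpawnDropPodForRankUp, which takes a Fortnite Athena player pawn and the new rank as a gameplay tag.
; The interface holds no logic, so spawning behavior is left to implementers.
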(fn SpawnDropPodForRankUp
  (params (PlayerPawn :type Fort%20Player%20Pawn%20Athena%20Object%20Reference) (NewRank :type Gameplay%20Tag%20Structure)))
