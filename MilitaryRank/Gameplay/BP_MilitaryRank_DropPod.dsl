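; Military rank drop pod, a child of a parent drop pod, whose rank sets its loot tier, marker icon and look.
; The server sets the rank; that setter and its OnRep share one function updating materials, tints and rank FX, with top ranks getting gold premium visuals, scale and destruction radius.
; ScalableFloat hotfixes gate a client-only pre-fall ground flare and timers that kill the pod, without death-opening, after spawn or opening.
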(event Custom|SpawnFlare
  (bind _flarespawnlocation (Variables|Default|GetFlareSpawnLocation))
  ; Spawn Flare
  (Variables|Default|SetFlareSpawnLocation (Utilities|Operators|Subtract (Transformation|GetActorLocation) (Math|Vector|MakeVector 0.0 0.0 (- (ScalableFloat|GetValueatLevel (Variables|Hotfix|GetHotfix_StartingHeight)) 8.0))))
  (Networking|SwitchHasAuthority
    (:Remote
      (Audio|PlaySoundatLocation (Variables|Default|GetFlareSpawnSound) _flarespawnlocation)
      (bind _returnvalue (Niagara|SpawnSystematLocation "/DropPodParentGameplay/Assets/FX/NS_TidalCrane_GroundBeam.NS_TidalCrane_GroundBeam" _flarespawnlocation))
      (bind _output_get (Variables|Default|SetSpawnFlareFX _returnvalue))
      (Niagara|SetNiagaraVariable(Float) _output_get "Projectile_Height" (ScalableFloat|GetValueatLevel (Variables|Hotfix|GetHotfix_StartingHeight)))
      (Utilities|Time|SetTimerbyEvent (delegate Custom|DeactivateFlareFX) (ScalableFloat|GetValueatLevel (Variables|Hotfix|GetHotfix_Flare_GroundFXLifespan))))))

(event EventBeginPlay
  ; Begin Play
  (|CallParentFunction)
  (Networking|SwitchHasAuthority
    (:Authority
      (if (> (ScalableFloat|GetValueatLevel (Variables|Hotfix|GetHotfix_Lifespan)) 0)
        (Utilities|Time|SetTimerbyEvent (delegate Custom|LifetimeEnded) (ScalableFloat|GetValueatLevel (Variables|Hotfix|GetHotfix_Lifespan)))))))

(event Custom|DeactivateFlareFX
  ; Spawn Flare
  (Components|Activation|Deactivate (Variables|Default|GetSpawnFlareFX)))

(event Custom|LifetimeEnded
  ; Die on lifetime expired
  (Variables|DropPodSettings|SetbOpenOnDeath false)
  (Gameplay|Die))

(event Custom|Server_SetAndRepDropPodRank
  (params (NewDropPodRank :type Integer))
  ; Server Set and Rep Drop Pod Rank
  (Variables|RankInfo|SetDropPodRank NewDropPodRank)
  (CallFunction|SetDropPodRankProperties))

(event EventSetStatePreFall
  ; Pre-fall state
  (|CallParentFunction)
  (if (ScalableFloat|GetValueasBool (Variables|Hotfix|GetHotfix_Flare_Enabled))
    (CallFunction|SpawnFlare)))

(event EventSetStateLandedOpened
  ; Opened state
  (|CallParentFunction)
  (Components|Activation|Deactivate (Variables|Default|GetNS_TidalCrane_Rank_UI)))

(event EventOnOpened
  ; On Opened
  (|CallParentFunction)
  (if (> (ScalableFloat|GetValueatLevel (Variables|Hotfix|GetHotfix_LifespawnAfterOpened)) 0)
    (Utilities|Time|SetTimerbyEvent (EventDispatchers|CreateEvent :!event Custom|LifetimeEnded) (ScalableFloat|GetValueatLevel (Variables|Hotfix|GetHotfix_LifespawnAfterOpened)))))

(fn AudioTimerSet
  (params (PlayState :type EAudioComponentPlayState%20Enum :default "Playing")))

(fn OnRep_DropPodRank
  (CallFunction|SetDropPodRankProperties))

(fn SetDropPodRankProperties
  (bind _droppodrank (Variables|RankInfo|GetDropPodRank))
  (bind _rank_emissivecolor (Variables|RankInfo|GetRank_EmissiveColor))
  (bind _output (Utilities|Array|Get(acopy) _rank_emissivecolor _droppodrank))
  (bind (_markerdisplay_displayname _markerdisplay_icon _markerdisplay_customindicatorclass _markerdisplay_customdetailsclass _markerdisplay_primarycolor _markerdisplay_sound _markerdisplay_battachsound _markerdisplay_screenclamping _markerdisplay_specialcull _markerdisplay_specialdotsizethreshold _markerdisplay_specialdistance _markerdisplay_customtimetolive _markerdisplay_additionaldata) (Utilities|Struct|BreakMarkedActorDisplayInfo (Variables|GetMarkerDisplay_DisplayName)))
  (bind _ns_tidalcrane_rank_ui (Variables|Default|GetNS_TidalCrane_Rank_UI))
  (Variables|DropPodSettings|SetLoot%20Tier%20Name (Utilities|Array|Get(acopy) (Variables|RankInfo|GetRank_LTG) (Variables|RankInfo|GetDropPodRank)))
  (Variables|SetMarkerDisplay_DisplayName (Utilities|Struct|MakeMarkedActorDisplayInfo _markerdisplay_displayname (Utilities|Array|Get(acopy) (Variables|RankInfo|GetRank_Icon) (Variables|RankInfo|GetDropPodRank)) _markerdisplay_customindicatorclass _markerdisplay_customdetailsclass (Utilities|Array|Get(acopy) (Variables|RankInfo|GetRank_IconColor) (Variables|RankInfo|GetDropPodRank)) _markerdisplay_sound true _markerdisplay_screenclamping _markerdisplay_specialcull _markerdisplay_specialdotsizethreshold _markerdisplay_specialdistance _markerdisplay_customtimetolive _markerdisplay_additionaldata))
  ; TODO - rank S+ is unsupported - remove clamp when working
  (Effects|Components|ParticleSystem|SetFloatParameter (Variables|Default|GetNS_TidalCrane_Rank_UI) "Rank" (Math|Conversions|ToFloat(Integer) (+ (Variables|RankInfo|GetDropPodRank) 1)))
  (if (>= (Variables|RankInfo|GetDropPodRank) 3)
    (Rendering|Material|SetMaterial (Variables|Debug|GetStaticMeshComponent) 0 (Variables|RankInfo|GetMaterial_Premium_Base))
    (Rendering|Material|SetMaterial (Variables|Debug|GetStaticMeshComponent) 1 (Variables|RankInfo|GetMaterial_Premium_Glass))
    (Transformation|SetWorldScale3D (Variables|Debug|GetStaticMeshComponent) (Math|Conversions|ToVector(Float) (Variables|RankInfo|GetPremiumScale)))
    (else
      (Rendering|Material|SetMaterial (Variables|Debug|GetStaticMeshComponent) 0 (Variables|Mesh|GetMaterialDropPodBase))
      (Rendering|Material|SetMaterial (Variables|Debug|GetStaticMeshComponent) 1 (Variables|Mesh|GetMaterialDropPodWindow))))
  (bind _returnvalue (CallFunction|GetorCreateMIDOnDefaultMesh 0))
  (bind _output_get (Variables|Default|SetMID_BaseMesh _returnvalue))
  (Rendering|Material|SetVectorParameterValue _output_get "EmissiveTint" _output :!class /Script/Engine.MaterialInstanceDynamic)
  (Rendering|Material|SetVectorParameterValue _output_get "Tint_Color" _output :!class /Script/Engine.MaterialInstanceDynamic)
  (Rendering|Material|SetScalarParameterValue _output_get "Rank" (Math|Conversions|ToFloat(Integer) (+ (Variables|RankInfo|GetDropPodRank) 1)) :!class /Script/Engine.MaterialInstanceDynamic)
  (Rendering|Material|SetVectorParameterValue _output_get "RankTint" _output :!class /Script/Engine.MaterialInstanceDynamic)
  (bind _returnvalue_2 (CallFunction|GetorCreateMIDOnDefaultMesh 1))
  (bind _output_get_2 (Variables|Default|SetMID_BaseMesh_Glass _returnvalue_2))
  (Rendering|Material|SetVectorParameterValue _output_get_2 "Tint_Color" _output :!class /Script/Engine.MaterialInstanceDynamic)
  (Rendering|Material|SetVectorParameterValue _output_get_2 "GlassDiffuse" _output :!class /Script/Engine.MaterialInstanceDynamic)
  (Rendering|Material|SetVectorParameterValue _output_get_2 "Tint_Color" _output :!class /Script/Engine.MaterialInstanceDynamic)
  ;[0.02753 0.02753 0.02753 0.5] FX RANKING SETUP
  (if (ScalableFloat|GetValueasBool (Variables|Hotfix|GetHotfix_ShowRankLetter))
    (Niagara|SetNiagaraVariable(Float) _ns_tidalcrane_rank_ui "Rank" (Math|Conversions|ToFloat(Integer) (+ (Variables|RankInfo|GetDropPodRank) 1)))
    (else
      (Components|Activation|SetActive _ns_tidalcrane_rank_ui false false)))
  ;;[1 0.4223 0.002847 0.5] GOLD GLOW AND SHIMMER ONLY FOR S+ RANK
  (if (>= (Variables|RankInfo|GetDropPodRank) 3)
    (Variables|DropPodSettings|SetbIsPodGold true)
    ;[0.1194 0.06413 0.15 0.5] Update imminent destruction radius with S+ size
    (Variables|/Game/Athena/Items/Components/ImminentDestructionComponent.ImminentDestructionComponent_C|SetHotfix_ImminentDestructionRadius (Variables|Hotfix|GetHotfix_PremiumImminentDestructionRadius) (Variables|Default|GetImminentDestructionComponent))))

(fn SetupDropPodMaterials)

(fn ConstructionScript
  (|CallParentFunction))
