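# EQS query that picks a drop pod spawn location around the querier, using a wide cone of navmesh-projected points.
# Filters enforce minimum distance, flat ground, an obstruction check and a tall capsule overlap check of the airspace above.
# Safe-zone and dot-product tests filter and score, while an inverse-linear distance score favors nearer points.
# Cone range and distance limits are bound to scalable-float data providers.
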EnvQuery /MilitaryRank/Gameplay/EQS_DropPodSpawnLocation
options [{
  generator {
   name=EnvQueryGenerator_Cone_0
   class=EnvQueryGenerator_Cone
   properties {
    alignedPointsDistance=200
    angleStep=10
    bAutoSortTests=true
    bCanRunAsync=false
    bIncludeContextLocation=false
    centerActor=&EnvQueryContext_Querier
    coneDegrees=170
    envQueryResultNormalizationOption=Default
    navDataOverrideContext=null
    optionName=""
    projectionData {
     bCanDisableTrace=true
     bCanProjectDown=true
     bCanTraceOnGeometry=true
     bCanTraceOnNavMesh=true
     bOnlyBlockingHits=true
     bTraceComplex=false
     extentX=0
     extentY=10
     extentZ=10
     navigationFilter=null
     postProjectionVerticalOffset=50
     projectDown=1024
     projectUp=1024
     serializedChannel=ECC_Visibility
     traceChannel=TraceTypeQuery1
     traceMode=Navigation
     traceProfileName=None
     traceShape=Line
     versionNum=1
    }
    range {
     dataBinding=&self:EnvQueryGenerator_Cone_0.FortAIDataProvider_ScalableFloat_0
     dataField=FloatValue
     defaultValue=1000
    }
   }
  }
  tests [
   FortQueryTest_InsideAthenaSafeZone_0 {class=FortQueryTest_InsideAthenaSafeZone, properties {testPurpose=FilterAndScore}}
   EnvQueryTest_Distance_1 {
    class=EnvQueryTest_Distance
    properties {
     distanceTo=&EnvQueryContext_Querier
     filterType=Minimum
     scoringEquation=Linear
     scoringFactor=1
     testComment=""
     testMode=Distance2D
     testPurpose=Filter
     floatValueMin {
      dataBinding=&self:EnvQueryTest_Distance_1.FortAIDataProvider_ScalableFloat_0
      dataField=FloatValue
      defaultValue=0
     }
    }
   }
   FortQueryTest_IsObstructed_0 {
    class=FortQueryTest_IsObstructed
    properties {
     context=&EnvQueryContext_Querier
     testPurpose=Filter
     traceData {serializedChannel=ECC_GameTraceChannel11, traceChannel=TraceTypeQuery10}
    }
   }
   FortQueryTest_OnFlatSurface_0 {class=FortQueryTest_OnFlatSurface, properties {radius=150, toleranceZ=30, traceOffsetUp=1000}}
   EnvQueryTest_Distance_2 {
    class=EnvQueryTest_Distance
    properties {
     distanceTo=&EnvQueryContext_Querier
     filterType=Range
     floatValueMin=0
     scoringEquation=InverseLinear
     scoringFactor=1
     testComment=""
     testMode=Distance2D
     testPurpose=Score
     floatValueMax {
      dataBinding=&self:EnvQueryTest_Distance_2.FortAIDataProvider_ScalableFloat_0
      dataField=FloatValue
      defaultValue=0
     }
    }
   }
   EnvQueryTest_Dot_1 {
    class=EnvQueryTest_Dot
    properties {
     clampMaxType=None
     clampMinType=None
     filterType=Minimum
     floatValueMax=1
     floatValueMin=0.6
     scoreClampMax=0
     scoreClampMin=0
     scoringFactor=1
     testComment=""
     testPurpose=FilterAndScore
    }
   }
   EnvQueryTest_Overlap_0 {
    class=EnvQueryTest_Overlap
    properties {overlapData {extentX=140, extentY=140, extentZ=10000, overlapChannel=ECC_GameTraceChannel4, overlapShape=Capsule, shapeOffset=<0 0 10000>}}
   }
  ]
 }]
